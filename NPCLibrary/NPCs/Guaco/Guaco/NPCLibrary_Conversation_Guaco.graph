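# Conversation database for the NPC Guaco, defining the greeting shown when a player starts talking to this taco-themed character.
# The entry point leads to one speech node with a default line and contextual overrides for a first-ever conversation or specific player character skins.
# First-meeting priority comes from a named curve row, while the skin-specific lines use a flat value.
# The speech node then links out to the NPC library quests and basic services conversation entries.
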ConversationDatabase /NPCLibrary/NPCs/Guaco/Guaco/NPCLibrary_Conversation_Guaco
entry_points [{tag=Conversation.Entry.Guaco, destinations [#1]}]
nodes [
 ConversationEntryPointNode_7 {
  id=#1
  class=ConversationEntryPointNode
  outputs [#2]
  properties {entryTag {tagName=Conversation.Entry.Guaco}}
 }
 Speech_C_4 {
  id=#2
  class=Speech_C
  outputs [#3, #4]
  properties {
   nodeName=""
   generalConfig {
    defaultMessage="Sorry, just sold the last taco."
    contextualMessages [
     {
      message=I'm not the mascot. I'm the general manager.
      contextRequirements [{requirement {class=FortControllerRequirement_IsFirstEverConversationWithNPC}}]
      priority {curve {rowName=Tandem.Speech.Priority.FirstEverInteraction}}
     }
     {
      message=Stop looking at my head like that...
      contextRequirements [{requirement {
         class=FortControllerRequirement_HasCID
         properties {allowedCIds [/BRCosmetics/Athena/Items/Cosmetics/Characters/CID_419_Athena_Commando_M_Pug]}
        }}]
      priority {value=0, curve {curveTable=null, rowName=None}}
     }
     {
      message=This shouldn't be happening. We won...
      requirementMatchPolicy=RequireAny
      contextRequirements [{requirement {
         class=FortControllerRequirement_HasCID
         properties {allowedCIds [
           /BRCosmetics/Athena/Items/Cosmetics/Characters/CID_338_Athena_Commando_M_DumplingMan
           /BRCosmetics/Athena/Items/Cosmetics/Characters/CID_344_Athena_Commando_M_IceCream
           /BRCosmetics/Athena/Items/Cosmetics/Characters/CID_109_Athena_Commando_M_Pizza
          ]}
        }}]
      priority {value=0, curve {curveTable=null, rowName=None}}
     }
     {
      message="You too cool for guac, kitty cat?"
      contextRequirements [{requirement {class=FortControllerRequirement_HasCID, properties {allowedCIds [&Character_ChillCat]}}}]
      priority {value=0, curve {curveTable=null, rowName=None}}
     }
     {
      message=You have tacos at that Academy of yours?
      contextRequirements [{requirement {class=FortControllerRequirement_HasCID, properties {allowedCIds [&Character_PinkSpike]}}}]
      priority {value=0, curve {curveTable=null, rowName=None}}
     }
     {
      message="I swear, when I called you Lil Bearito Taquito, I was just teasing!"
      contextRequirements [{requirement {class=FortControllerRequirement_HasCID, properties {allowedCIds [&Character_BadBear]}}}]
      priority {value=0, curve {curveTable=null, rowName=None}}
     }
     {
      message=Just seeing you makes me want to retire.
      contextRequirements [{requirement {class=FortControllerRequirement_HasCID, properties {allowedCIds [&Character_TheHerald]}}}]
      priority {value=0, curve {curveTable=null, rowName=None}}
     }
     {
      message="Y'know, I could use a delivery person..."
      contextRequirements [{requirement {class=FortControllerRequirement_HasCID, properties {allowedCIds [&Character_RoseDust]}}}]
      priority {value=0, curve {curveTable=null, rowName=None}}
     }
    ]
   }
  }
 }
 ConversationLinkNode_0 {id=#4, class=ConversationLinkNode, properties {remoteEntryTag {tagName=Conversation.Entry.BasicServices}}}
 ConversationLinkNode_1 {id=#3, class=ConversationLinkNode, properties {remoteEntryTag {tagName=Conversation.Entry.NPCLibrary.Quests}}}
]
